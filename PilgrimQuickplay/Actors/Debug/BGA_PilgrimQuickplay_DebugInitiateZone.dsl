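; Debug interactable actor that starts a Pilgrim Quickplay team music battle for the interacting player's squad.
; Interaction is allowed only with no battle in progress and when the pawn lacks the PlayingMusic tag.
; Activation gathers squad pawns without an active MusicBattle ability, appends their player states, then spawns a TeamQuickplay MusicBattleInstance at the interactor.
; The battle's finished event destroys the instance and clears the reference; PlayerStates is never cleared here.
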(event EventBeginPlay)

(event Collision|EventActorBeginOverlap
  (params (OtherActor :type Actor%20Object%20Reference)))

(event EventTick
  (params (DeltaSeconds :type Float%20%28single-precision%29)))

(event Interaction|EventBlueprintOnInteract
  (params (InteractingPawn :type Fort%20Pawn%20Object%20Reference) (InteractionBeingAttempted :type EInteractionBeingAttempted%20Enum :default "FirstInteraction"))
  (CallFunction|ActivateQuickplayBattleInstance InteractingPawn))

(fn ActivateQuickplayBattleInstance
  (params (Interacting%20Pawn :type Fort%20Pawn%20Object%20Reference))
  (Utilities|FlowControl|Sequence
    (:then_0
      ;[1 1 1] Get all players in the squad
      (bind _asfort_player_pawn_athena (Utilities|Casting|CastToFortPlayerPawnAthena Interacting%20Pawn))
      (bind _returnvalue (Fortnite|Utilities|AttemptGetPlayerStateforPawn _asfort_player_pawn_athena))
      (bind _asfort_player_state_athena (Utilities|Casting|CastToFortPlayerStateAthena _returnvalue))
      (bind _returnvalue_2 (Class|FortGameStateAthena|GetSquadMemberPlayerStates (Fortnite|Athena|GameState|GetGameStateAthena) _asfort_player_state_athena true))
      (for _array_element _returnvalue_2
        (bind _asfort_player_pawn_athena_2 (Utilities|Casting|CastToFortPlayerPawnAthena (PlayerState|GetPawn _array_element)))
        (if (not (Abilities|HasActiveAbilitywithAnyMatchingTag (Ability|GetFortAbilitySystemComponent _asfort_player_pawn_athena_2) "(GameplayTags=((TagName=\"Mojo.Pilgrim.Abilities.MusicBattle\")))"))
          (Utilities|Array|Add (LocalVariables|GetSquadPawns) _asfort_player_pawn_athena_2))))
    (:then_1
      ;[1 1 1] Create the MusicBattleInstance
      (Variables|Default|GetInProgressBattle
        (:then)
        (:else
          (CallFunction|BuildPlayerStates (LocalVariables|GetSquadPawns))
          (bind _returnvalue_3 (Game|SpawnActorfromClass "/PilgrimQuickplay/Actors/MusicBattle/MusicBattleInstance_TeamQuickplay.MusicBattleInstance_TeamQuickplay_C" (Transformation|GetActorTransform Interacting%20Pawn) "Undefined" "OverrideRootScale" (Fortnite|GetFortPlayerControllerfromActor Interacting%20Pawn) (Variables|Default|GetPlayerStates) "TeamQuickplay" Interacting%20Pawn))
          (bind _output_get (Variables|Default|SetInProgressBattle _returnvalue_3))
          (EventDispatchers|BindEventtoOnMusicBattleFinished _output_get (EventDispatchers|CreateEvent :!event OnMusicBattleFinished)))))))

(fn BlueprintCanInteract
  (params (InteractingPawn :type Fort%20Pawn%20Object%20Reference) (InteractionBeingAttempted :type EInteractionBeingAttempted%20Enum :default "FirstInteraction") (InteractionType :type TInteractionType%20Enum :default "IT_NoInteraction"))
  (return (and (not (Utilities|IsValid (Variables|Default|GetInProgressBattle) :!class /Script/Engine.KismetSystemLibrary)) (GameplayTags|DoesContainerMatchTagQuery (GameplayTags|GetOwnedGameplayTags InteractingPawn :!class /Script/GameplayTags.GameplayTagAssetInterface) (GameplayTags|MakeGameplayTagQuery "(TokenStreamVersion=0,TagDictionary=((TagName=\"Granted.Mojo.Pilgrim.State.PlayingMusic\")),QueryTokenStream=(0,1,5,1,3,1,0),UserDescription=\"\",AutoDescription=\" ALL(  NONE( Granted.Mojo.Pilgrim.State.PlayingMusic ) )\")")))))

(fn BlueprintGetInteractionString
  (params (InteractingPawn :type Fort%20Pawn%20Object%20Reference) (InteractionBeingAttempted :type EInteractionBeingAttempted%20Enum :default "FirstInteraction"))
  (return (Variables|Default|GetInteractText)))

(fn BlueprintGetInteractionTime
  (params (InteractingPawn :type Fort%20Pawn%20Object%20Reference) (InteractionBeingAttempted :type EInteractionBeingAttempted%20Enum :default "FirstInteraction"))
  (return (Variables|Default|GetInteractTime) true))

(fn BuildPlayerStates
  (params (PlayerPawns :type Array%20of%20Fort%20Player%20Pawn%20Athena%20Object%20References))
  (for _array_element PlayerPawns
    (bind _asfort_player_state_athena (Utilities|Casting|CastToFortPlayerStateAthena (Variables|/Script/Engine.Pawn|GetPlayerState _array_element)))
    (Utilities|Array|Add (Variables|Default|GetPlayerStates) _asfort_player_state_athena)))

(fn OnMusicBattleFinished
  (params (EndReason :type EMusicBattleEndReason%20Enum :default "CompletedNormally"))
  (bind _inprogressbattle (Variables|Default|GetInProgressBattle)
    (:then
      (Actor|DestroyActor _inprogressbattle)
      (Variables|Default|SetInProgressBattle 0))
    (:else)))

(fn ConstructionScript)
