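; Debug interactable device for Pilgrim Quickplay that sets the interacting player's song vote to its configured SongToVoteFor.
; Interaction is allowed only when the pawn lacks the PlayingMusic gameplay tag, with prompt text and hold time read from variables.
; On interact, SetVoteForPlayer finds the PilgrimQuickplayPlayerStateComponent through the pawn's PlayerState, sets the vote and prints a debug confirmation.
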(event EventBeginPlay)

(event Collision|EventActorBeginOverlap
  (params (OtherActor :type Actor%20Object%20Reference)))

(event EventTick
  (params (DeltaSeconds :type Float%20%28single-precision%29)))

(event Interaction|EventBlueprintOnInteract
  (params (InteractingPawn :type Fort%20Pawn%20Object%20Reference) (InteractionBeingAttempted :type EInteractionBeingAttempted%20Enum :default "FirstInteraction"))
  (CallFunction|SetVoteforPlayer InteractingPawn))

(fn BlueprintCanInteract
  (params (InteractingPawn :type Fort%20Pawn%20Object%20Reference) (InteractionBeingAttempted :type EInteractionBeingAttempted%20Enum :default "FirstInteraction") (InteractionType :type TInteractionType%20Enum :default "IT_NoInteraction"))
  (return (GameplayTags|DoesContainerMatchTagQuery (GameplayTags|GetOwnedGameplayTags InteractingPawn :!class /Script/GameplayTags.GameplayTagAssetInterface) (GameplayTags|MakeGameplayTagQuery "(TokenStreamVersion=0,TagDictionary=((TagName=\"Granted.Mojo.Pilgrim.State.PlayingMusic\")),QueryTokenStream=(0,1,5,1,3,1,0),UserDescription=\"\",AutoDescription=\" ALL(  NONE( Granted.Mojo.Pilgrim.State.PlayingMusic ) )\")"))))

(fn BlueprintGetInteractionString
  (params (InteractingPawn :type Fort%20Pawn%20Object%20Reference) (InteractionBeingAttempted :type EInteractionBeingAttempted%20Enum :default "FirstInteraction"))
  (return (Variables|Default|GetInteractText)))

(fn BlueprintGetInteractionTime
  (params (InteractingPawn :type Fort%20Pawn%20Object%20Reference) (InteractionBeingAttempted :type EInteractionBeingAttempted%20Enum :default "FirstInteraction"))
  (return (Variables|Default|GetInteractTime) true))

(fn SetVoteForPlayer
  (params (InteractingPawn :type Fort%20Pawn%20Object%20Reference))
  (bind _asfort_player_pawn (Utilities|Casting|CastToFortPlayerPawn InteractingPawn))
  (bind _returnvalue (Class|PilgrimQuickplayPlayerStateComponent|GetPilgrimQuickplayPlayerStateComponent (Variables|/Script/Engine.Pawn|GetPlayerState _asfort_player_pawn)))
  (bind _output_get (LocalVariables|SetQuickplayPlayerStateComponent _returnvalue))
  (Utilities|IsValid _output_get
    (:Is%20Valid
      (Class|PilgrimQuickplayPlayerStateComponent|SetSongVote (LocalVariables|GetQuickplayPlayerStateComponent) (Variables|Default|GetSongToVoteFor))
      (Development|PrintText (Utilities|Text|FormatText "NSLOCTEXT(\"[675A23DAD54D4F108BCDD1177AA1B173]\", \"06B9723C426F169C53D12F960661507F\", \"{0} Song vote set to :{1}\")" (Utilities|GetDisplayName (Variables|GetInteractingPawn)) (Meta|GetShortName (Variables|Default|GetSongToVoteFor)))))))

(fn ConstructionScript)
